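# Battle Royale HUD widget that shows player action alerts such as downs, eliminations, storm entry and squad wipes.
# Show Alert maps alert type to a MessageSwitcher child and passes it the data; children's RequestNextAlert events advance alerts.
# Empty-state types, and crit kills unless a console variable is on, are skipped by hiding and requesting the next.
# Squad wipes play on a separate overlay widget via their own event.
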WidgetBlueprint /BattleRoyaleGameUI/HUD/PlayerActionAlert/WBP_BRGameUI_PlayerActionAlert
graphs [EventGraph {nodes [
   n1 = HandleRequestNextAlert() {is=title K2Node_CustomEvent}
   > RequestNextAlert(self:Athena HUDPlayer Action Alert Base Object Reference)
   Event Construct()
   n2 = GetAllChildren(self=Get MessageSwitcher() -> (MessageSwitcher:Common Visibility Switcher Object Reference))
   > n3 = For Each Loop(Array:Array of Widget Object References=@n2) -> (Array Element:Widget Object Reference)
   > n4 = Cast To WBP_BRGameUI_ActionAlertBase!(execute, Object=@n3) -> (then, CastFailed, AsWBP BRGame UI Action Alert Base:WBP BRGame UI Action Alert Base Object Reference) {is=title K2Node_DynamicCast}
   > Bind Event to Event Request Next Alert(self:WBP BRGame UI Action Alert Base Object Reference=@n4, Delegate=@n1) {is=title K2Node_AddDelegate}
   Event Destruct()
   > n5 = For Each Loop(Array=@n2)
   > n6 = Cast To WBP_BRGameUI_ActionAlertBase(Object=@n5)
   > Unbind all Events from Event Request Next Alert!(execute:Exec, self:WBP BRGame UI Action Alert Base Object Reference=@n6) -> (then:Exec) {is=title K2Node_ClearDelegate}
   Event On Squad Wipe() {is=title K2Node_Event}
   > ShowAlert(self:WBP BRGame UI Squad Elimination Alert Object Reference=Get WBP_BRGameUI_SquadEliminationAlert() -> (WBP_BRGameUI_SquadEliminationAlert:WBP BRGame UI Squad Elimination Alert Object Reference))
   Event On Alerts Cleared() {is=title K2Node_Event}
   > SetVisibility(InVisibility=Hidden)
   n7 = Event Show Alert!() -> (OutputDelegate, then, Alert:Athena Player Action Alert Structure) {is=title K2Node_Event}
   > n8 = GetConsoleVariableBoolValue!(execute, VariableName:String=HUDPlayerActionAlert.HandleAllEliminationAlertsThroughPawnDeathEvent) -> (then, ReturnValue:bool)
   n9 = Break Athena Player Action Alert!(AthenaPlayerActionAlert:"Athena Player Action Alert Structure (by ref)"=@n7.Alert) -> (
    AlertType:EAthenaPlayerActionAlert Enum
    DetailText:Text
    RankedProgress:Ranked Progress Replicated Data Structure
    OtherPlayerState:Fort Player State Athena Weak object
    NumKills:int
   ) {is=title K2Node_BreakStruct}
   n10 = "Not Equal (Enum)"!(A:EAthenaPlayerActionAlert Enum=@n9, B:EAthenaPlayerActionAlert Enum=PlayerKillCritKill) -> (ReturnValue:bool)
   n11 = Select!(
    PlayerDown:int=1
    PlayerKill:int=2
    EnteredStorm:int=3
    NewZebulonDrone:int=4
    NewZebulonDroneYou:int=4
    PlayerKillSquadWipe:int=0
    PlayerKillCritKill:int=2
    Index:EAthenaPlayerActionAlert Enum=@n9
   ) -> (ReturnValue:int)
   > n12 = Branch(Condition=AND Boolean(A=OR Boolean(A=@n8, B=@n10), B="integer > integer"(A=@n11)))
   > SetVisibility(InVisibility=SelfHitTestInvisible)
   n13 = Get MessageSwitcher()
   > SetActiveWidgetIndex(self:Common Visibility Switcher Object Reference=@n13, Index=@n11)
   > n14 = Cast To WBP_BRGameUI_ActionAlertBase(Object=GetActiveWidget!(self:Common Visibility Switcher Object Reference=@n13) -> (ReturnValue:Widget Object Reference))
   > ShowAlert(self:WBP BRGame UI Action Alert Base Object Reference=@n14)
   > HandleAlertData!(execute, self:WBP BRGame UI Action Alert Base Object Reference=@n14, AlertData:Athena Player Action Alert Structure=@n7.Alert) -> (then)
   SetVisibility(execute=@n12.else, InVisibility=Hidden)
   > RequestNextAlert()
  ]}]
variables [
 OtherPlayerPawn {type=FortPlayerPawn}
 RankData {type=FortHabaneroDisplayData, default=/Game/UI/Habanero/HabaneroDisplayData}
 SecondaryInfoContent {type=UserWidget, default=null}
]
widgets {
 info {inheritedWidgetCount=1, parentClass=/Script/FortniteUI.AthenaHUDPlayerActionAlertBase, rootWidgetClass=&Overlay, widgetCount=9}
 widgets [
  Overlay_0 {widgetClassPath=&Overlay, widget=Overlay_0, properties {bOverride_Cursor=true}}
  WBP_BRGameUI_SquadEliminationAlert {
   widgetClassPath=&/WBP_BRGameUI_SquadEliminationAlert
   bIsVariable=true
   parent=Overlay_0
   widget=WBP_BRGameUI_SquadEliminationAlert
   properties {visibility=HitTestInvisible, renderTransform {angle=0, scale=<1 1>, shear=<0 0>, translation=<0 -90>}}
   slot {class=OverlaySlot, properties {horizontalAlignment=HAlign_Center}}
  }
  MessageSwitcher {
   widgetClassPath=&CommonVisibilitySwitcher
   bIsVariable=true
   parent=Overlay_0
   widget=MessageSwitcher
   properties {activeWidgetIndex=2}
   slot {class=OverlaySlot}
  }
  Overlay_EmptyState {widgetClassPath=&Overlay, parent=MessageSwitcher, widget=Overlay_EmptyState, slot {class=CommonVisibilitySwitcherSlot}}
  WBP_BRGameUI_DownedAlert {
   widgetClassPath=&/WBP_BRGameUI_DownedAlert
   bIsVariable=true
   parent=MessageSwitcher
   widget=WBP_BRGameUI_DownedAlert
   slot {class=CommonVisibilitySwitcherSlot}
  }
  WBP_BRGameUI_EliminationAlert {
   widgetClassPath=&/WBP_BRGameUI_EliminationAlert
   bIsVariable=true
   parent=MessageSwitcher
   widget=WBP_BRGameUI_EliminationAlert
   properties {bPreventSequentialTriggering=true}
   slot {class=CommonVisibilitySwitcherSlot}
  }
  WBP_BRGameUI_StormAlert {
   widgetClassPath=&/WBP_BRGameUI_StormAlert
   bIsVariable=true
   parent=MessageSwitcher
   widget=WBP_BRGameUI_StormAlert
   slot {class=CommonVisibilitySwitcherSlot}
  }
  WBP_BRGameUI_SkydiveLeaderAlert {
   widgetClassPath=&/WBP_BRGameUI_SkydiveLeaderAlert
   bIsVariable=true
   parent=MessageSwitcher
   widget=WBP_BRGameUI_SkydiveLeaderAlert
   slot {class=CommonVisibilitySwitcherSlot}
  }
  TutorialHighlight {widgetClassPath=&TutorialHighlight}
 ]
}
