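# HUD extension container for the Battle Royale UI, providing the slot area above the storm message, built on FortExtensionWidgetContainer.
# A root vertical box holds a built-in player action alert inside a fixed-size SizeBox, aligned top-center.
# When an extension widget is created, it is appended to that box and its slot is set to fill both axes, with no filtering by slot type.
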WidgetBlueprint /BattleRoyaleGameUI/HUD/WBP_BRGameUI_ExtensionSlot_AboveStormMessage
graphs [EventGraph {nodes [
   n1 = Event OnExtensionWidgetCreated!() -> (OutputDelegate, then, InUserWidget:User Widget Object Reference, InExtensionSlot:EUIExtensionSlot Enum=Invalid) {is=title K2Node_Event}
   > n2 = AddChild(self=Get AboveStormMessageSlotBox() -> (AboveStormMessageSlotBox:Vertical Box Object Reference), Content=@n1.InUserWidget)
   > n3 = Cast To VerticalBoxSlot!(execute, Object=@n2) -> (then, CastFailed, AsVertical Box Slot:Vertical Box Slot Object Reference) {is=title K2Node_DynamicCast}
   > SetVerticalAlignment(self:Vertical Box Slot Object Reference=@n3, InVerticalAlignment=VAlign_Fill)
   > SetHorizontalAlignment(self:Vertical Box Slot Object Reference=@n3, InHorizontalAlignment=HAlign_Fill)
  ]}]
variables []
widgets {
 info {inheritedWidgetCount=0, parentClass=/Script/FortniteUI.FortExtensionWidgetContainer, rootWidgetClass=&VerticalBox, widgetCount=3}
 widgets [
  AboveStormMessageSlotBox {widgetClassPath=&VerticalBox, bIsVariable=true, widget=AboveStormMessageSlotBox}
  SizeBox_0 {
   widgetClassPath=&SizeBox
   parent=AboveStormMessageSlotBox
   widget=SizeBox_0
   properties {heightOverride=128, widthOverride=500}
   slot {class=VerticalBoxSlot}
  }
  WBP_BRGameUI_PlayerActionAlert {
   widgetClassPath=/BattleRoyaleGameUI/HUD/PlayerActionAlert/WBP_BRGameUI_PlayerActionAlert
   parent=SizeBox_0
   widget=WBP_BRGameUI_PlayerActionAlert
   slot {class=SizeBoxSlot, properties {horizontalAlignment=HAlign_Center, verticalAlignment=VAlign_Top}}
  }
 ]
}
