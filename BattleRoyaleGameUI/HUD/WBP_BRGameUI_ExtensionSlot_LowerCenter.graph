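# Battle Royale HUD widget that acts as the lower-center extension slot for dynamically added UI content.
# On Construct it registers as a player-specific widget host under the LowerCenter slot gameplay tag, and unregisters on Destruct.
# Add and remove host requests that exactly match that tag add or remove the content widget in a centered, zero-transition prioritized widget switcher.
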WidgetBlueprint /BattleRoyaleGameUI/HUD/WBP_BRGameUI_ExtensionSlot_LowerCenter
graphs [
 EventGraph {nodes [
   Event Construct()
   > Parent: Construct()
   > RegisterWidgetHost!(
    execute
    WidgetHost:Widget Object Reference
    WidgetHostTag:"Gameplay Tag Structure (by ref)"=Get BottomCenterTag() -> (BottomCenterTag:Gameplay Tag Structure)
    IsPlayerSpecific:bool=true
   ) -> (then)
   Event Destruct()
   > UnregisterWidgetHost!(execute, WidgetHost:Widget Object Reference, WidgetHostTag:"Gameplay Tag Structure (by ref)"=Get BottomCenterTag()) -> (then)
   n1 = Event Add Content To Host Widget!() -> (
    OutputDelegate
    then
    WidgetHostTag:"Gameplay Tag Structure (by ref)"
    ContentWidget:Widget Object Reference
    Padding:"Margin Structure (by ref)"
    bFill:bool
   ) {is=title K2Node_Event}
   > Branch(Condition=MatchesTag!(TagOne:Gameplay Tag Structure=@n1.WidgetHostTag, TagTwo:Gameplay Tag Structure=Get BottomCenterTag(), bExactMatch:bool=true) -> (ReturnValue:bool))
   > AddChild(
    self=Get WidgetSwitcher_BottomCenter() -> (WidgetSwitcher_BottomCenter:Fort Prioritized Widget Switcher Object Reference)
    Content=@n1.ContentWidget
   )
  ]}
 RemoveContentFromWidgetHost {nodes [
   n1 = Remove Content From Host Widget!() -> (
    then
    WidgetHostTag:"Gameplay Tag Structure (by ref)"
    ContentWidget:Widget Object Reference
    RemovalType:EWidgetSlottingRemovalType Enum=Unparent
   ) {is=title K2Node_FunctionEntry}
   > Branch(Condition=MatchesTag(TagOne=@n1, TagTwo=Get BottomCenterTag(), bExactMatch=true))
   > n2 = RemoveChild(self=Get WidgetSwitcher_BottomCenter(), Content=@n1.ContentWidget)
   > Return Node!(execute, ReturnValue:bool=@n2)
  ]}
]
variables [BottomCenterTag {type=GameplayTag, default {tagName=DynamicUI.HUD.Fort.Athena.LowerCenter.Slot}}]
widgets {
 info {inheritedWidgetCount=0, parentClass=&UserWidget, rootWidgetClass=&Overlay, widgetCount=3}
 widgets [
  Overlay_Root {widgetClassPath=&Overlay, widget=Overlay_Root, properties {bOverride_Cursor=true}}
  VerticalBox_0 {
   widgetClassPath=&VerticalBox
   parent=Overlay_Root
   widget=VerticalBox_0
   slot {class=OverlaySlot, properties {horizontalAlignment=HAlign_Center}}
  }
  WidgetSwitcher_BottomCenter {
   widgetClassPath=/Script/FortniteUI.FortPrioritizedWidgetSwitcher
   bIsVariable=true
   parent=VerticalBox_0
   widget=WidgetSwitcher_BottomCenter
   properties {transitionDuration=0}
   slot {class=VerticalBoxSlot, properties {horizontalAlignment=HAlign_Center, verticalAlignment=VAlign_Top}}
  }
 ]
}
